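# Default Lit placeholder-style material in the FX folder of the BitterSweet main-player emote.
# The graph holds a single pale pink vector parameter, with no connection shown to any material output.
# Usage flags enable it for skeletal meshes, instanced static meshes, mesh deformers and Niagara mesh particles.
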Material /BRCosmetics/Animation/Game/MainPlayer/Emotes/BitterSweet/FX/Materials/01___Default
graph {
 properties {
  bUsedWithInstancedStaticMeshes=true
  bUsedWithMeshDeformer=true
  bUsedWithNiagaraMeshParticles=true
  bUsedWithSkeletalMesh=true
  pixelDepthOffsetMode=PDOM_Legacy
  usedShadingModels=Default Lit
 }
 nodes [VectorParameter() {defaultValue=<rgba 0.9784 0.8427 0.9706 0>}]
}
