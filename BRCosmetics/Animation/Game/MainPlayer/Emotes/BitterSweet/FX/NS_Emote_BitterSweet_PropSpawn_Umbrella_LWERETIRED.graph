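# Retired Niagara system for the umbrella prop-spawn effect of the BitterSweet emote, marked LWERETIRED in its name.
# All three emitters (SparklesLWE, AirglowLWE, SmokeLWE) are disabled and list no renderers, so it draws nothing as saved.
# The system script holds only SystemState and a variable assignment, alongside a NormalizedNotifyProgress user float, presumably meant for anim-notify timing.
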NiagaraSystem /BRCosmetics/Animation/Game/MainPlayer/Emotes/BitterSweet/FX/NS_Emote_BitterSweet_PropSpawn_Umbrella_LWERETIRED
dependencies {
 usedDataInterfaces []
 usedDynamicInputs []
 usedModules [
  &self:SystemSpawnScript.SystemScriptSource.SystemScriptGraph.NiagaraNodeAssignment_0.SetVariables_#1
  &SystemState
 ]
}
emitters {
 AirglowLWE {bEnabled=false, rendererClasses [], renderers []}
 SmokeLWE {bEnabled=false, rendererClasses [], renderers []}
 SparklesLWE {bEnabled=false, rendererClasses [], renderers []}
}
summary {
 systemName=NS_Emote_BitterSweet_PropSpawn_Umbrella_LWERETIRED
 emitters [
  {bEnabled=false, emitterName=SparklesLWE, rendererClasses []}
  {bEnabled=false, emitterName=AirglowLWE, rendererClasses []}
  {bEnabled=false, emitterName=SmokeLWE, rendererClasses []}
 ]
 userVariables [User.NormalizedNotifyProgress {type=NiagaraFloat, defaultValue=0, description=""}]
}
