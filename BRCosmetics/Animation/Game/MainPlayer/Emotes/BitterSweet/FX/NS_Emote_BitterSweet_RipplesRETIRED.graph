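# Retired Niagara system for flat water ripples in the BitterSweet emote, with every emitter disabled.
# Two mesh emitters burst at System.SplashPosition with random orientation and water snapping: a short faint ripple and a longer one tinted by time of day.
# Curves drive scale, color fade and dynamic material parameters, while particles stay constrained to a plane with a camera offset.
# Two LWE emitters have no renderers, and user variables expose water plane, depth and wave data.
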NiagaraSystem /BRCosmetics/Animation/Game/MainPlayer/Emotes/BitterSweet/FX/NS_Emote_BitterSweet_RipplesRETIRED
dependencies {
 usedDataInterfaces [&NiagaraDataInterfaceVectorCurve, &NiagaraDataInterfaceCurve]
 usedDynamicInputs []
 usedModules [
  &self:SystemSpawnScript.SystemScriptSource.SystemScriptGraph.NiagaraNodeAssignment_1.SetVariables_#1
  &SystemState
  &self:SystemSpawnScript.SystemScriptSource.SystemScriptGraph.NiagaraNodeAssignment_0.SetVariables_#2
 ]
}
emitters {
 Flat_Ripple {
  bEnabled=false
  emitterUpdateScript [
   &EmitterState(Life Cycle Mode=::NewEnumerator1, Loop Behavior=::NewEnumerator1, Loop Delay=0.15, Loop Duration=1, UseLoopDelay=false) {enabled=false}
   &SpawnBurst_Instantaneous(Spawn Time=0.18, Use Spawn Probability=false, UseLoopCountLimit=false) {enabled=false}
  ]
  particleSpawnScript [
   &InitializeParticle(
    Color=<rgba 1 1 1 0.07>
    Lifetime=&RandomRangeFloat(Minimum=1, Maximum=1.2)
    Mesh Scale Max=<0.4 0.4 0.12>
    Mesh Scale Min=<0.4 0.4 0.1>
    Mesh Scale Mode=::NewEnumerator2
    Position Offset=$System.SplashPosition
    UsePositionOffset=true
    Write Material Random=true
   ) {module=InitializeParticle001}
   &InitialMeshOrientation(
    Mesh Orientation Mode=::NewEnumerator7
    Rotation=&RandomRangeVector(Minimum=<0 0 0>, Maximum=<0 0 1>)
    Use Rotation Vector=true
   )
   &InitializeWaterSnapToOwner()
  ]
  particleUpdateScript [
   &ParticleState()
   &ScaleMeshSize(Scale Factor=&VectorFromCurve(VectorCurve=DataInterface() {propertyValues {
       XCurve {keys [
         {interpMode=RCIM_Cubic, leaveTangent=7.6613, tangentMode=RCTM_Break, time=0, value=0}
         {arriveTangent=2.4936, interpMode=RCIM_Cubic, leaveTangent=2.4936, tangentMode=RCTM_User, time=0.2392, value=1.3545}
         <at 1 2>
        ]}
       YCurve {keys [
         {interpMode=RCIM_Cubic, leaveTangent=7.7916, tangentMode=RCTM_Break, time=0, value=0}
         {arriveTangent=2.6017, interpMode=RCIM_Cubic, leaveTangent=2.6017, tangentMode=RCTM_User, time=0.2392, value=1.3545}
         <at 1 2>
        ]}
       ZCurve {keys [<at 0 3>,<at 1 1>]}
      }}))
   &DynamicMaterialParameters(
    Index 0 Param 1=&FloatFromCurve(FloatCurve=DataInterface() {propertyValues {Curve {keys [<at 0 0>,<at 1 1.5>]}}})
    Index 0 Param 2=&FloatFromCurve(FloatCurve=DataInterface() {propertyValues {Curve {keys [<at 0 1>,<at 1 0.5>]}}}, Scale Curve=0.35)
    Index 0 Param 3=50
    Index 0 Param 4=1
    Module.Param0WriteEnabled=true
    Module.Param1WriteEnabled=true
    Module.Param2WriteEnabled=true
    Module.Param3WriteEnabled=true
   )
   &CameraOffset(Camera Offset Amount=5)
   &ScaleColor(
    Scale Alpha=&FloatFromCurve(FloatCurve=DataInterface() {propertyValues {Curve {keys [
         {arriveTangent=9.0725, interpMode=RCIM_Cubic, leaveTangent=9.0725, tangentMode=RCTM_User, time=0, value=0}
         {arriveTangent=1, interpMode=RCIM_Cubic, leaveTangent=1, time=0.1408, value=1}
         {arriveTangent=-1.1638, interpMode=RCIM_Cubic, leaveTangent=-1.1638, time=0.5, value=0.5}
         {interpMode=RCIM_Cubic, leaveTangent=-0.0405, time=1, value=0}
        ]}}})
    Scale RGB=&VectorFromFloat(Value=&FloatFromCurve(FloatCurve=DataInterface() {propertyValues {Curve {keys [<at 0 2>,<at 1 0.2>]}}}))
    ScaleA=true
    ScaleRGB=true
   )
   &ConstrainPositionToPlane(Plane Position:NiagaraPosition=$Particles.Initial.Position)
  ]
  rendererClasses [&NiagaraMeshRendererProperties]
  renderers [{rendererClass=&NiagaraMeshRendererProperties, rendererIndex=0}]
 }
 Flat_Ripple001 {
  bEnabled=false
  emitterUpdateScript [
   &EmitterState(Inactive Response=::NewEnumerator1, Life Cycle Mode=::NewEnumerator1, Loop Delay=0.15, Loop Duration=2, UseLoopDelay=false) {enabled=false}
   &SpawnBurst_Instantaneous(Use Spawn Probability=false, UseLoopCountLimit=false) {enabled=false}
  ]
  particleSpawnScript [
   &InitializeParticle(
    Color=&FNTimeOfDayColor(
     Night Color=<rgba 0.3759 0.4941 0.8802 1>
     Morning Color=<rgba 1 0.7835 0.7011 1>
     Day Color=<rgba 0.75 0.75 0.75 0.9391>
     Evening Color=<rgba 1 0.7331 0.625 1>
    )
    Lifetime=4
    Mesh Scale Max=<0.4 0.4 0.12>
    Mesh Scale Min=<0.4 0.4 0.1>
    Mesh Scale Mode=::NewEnumerator2
    Position Offset=$System.SplashPosition
    UsePositionOffset=true
    Write Material Random=true
   ) {module=InitializeParticle001}
   &InitialMeshOrientation(
    Mesh Orientation Mode=::NewEnumerator7
    Rotation=&RandomRangeVector(Minimum=<0 0 0>, Maximum=<0 0 1>)
    Use Rotation Vector=true
   )
   &InitializeWaterSnapToOwner()
  ]
  particleUpdateScript [
   &ParticleState()
   &ScaleMeshSize(Scale Factor=&VectorFromCurve(VectorCurve=DataInterface() {propertyValues {
       XCurve {keys [{interpMode=RCIM_Cubic, leaveTangent=1.7427, tangentMode=RCTM_Break, time=0, value=1.5}, <at 1 2.1>]}
       YCurve {keys [{interpMode=RCIM_Cubic, leaveTangent=1.8106, tangentMode=RCTM_Break, time=0, value=1.5}, <at 1 2.1>]}
       ZCurve {keys [<at 0 3>,<at 1 2>]}
      }}))
   &DynamicMaterialParameters(
    Index 0 Param 1=&FloatFromCurve(FloatCurve=DataInterface() {propertyValues {Curve {keys [<at 0 0.5>,<at 1 1>]}}})
    Index 0 Param 2=&FloatFromCurve(FloatCurve=DataInterface() {propertyValues {Curve {keys [<at 0 1>,<at 1 0.75>]}}}, Scale Curve=0.35)
    Index 0 Param 3=10
    Index 0 Param 4=0.8404
    Module.Param0WriteEnabled=true
    Module.Param1WriteEnabled=true
    Module.Param2WriteEnabled=true
    Module.Param3WriteEnabled=true
   )
   &CameraOffset(Camera Offset Amount=5)
   &ScaleColor(
    Scale Alpha=&FloatFromCurve(FloatCurve=DataInterface() {propertyValues {Curve {keys [
         {arriveTangent=1.9378, interpMode=RCIM_Cubic, leaveTangent=1.9378, tangentMode=RCTM_User, time=0, value=0}
         <at 0.5 0.9 RCIM_Cubic>
         {interpMode=RCIM_Cubic, leaveTangent=-0.0405, time=1, value=0}
        ]}}})
    Scale RGB=&VectorFromFloat(Value=&FloatFromCurve(FloatCurve=DataInterface() {propertyValues {Curve {keys [<at 0 2>,<at 1 0.2>]}}}))
    ScaleA=true
    ScaleRGB=true
   )
   &ConstrainPositionToPlane(Plane Position:NiagaraPosition=$Particles.Initial.Position)
  ]
  rendererClasses [&NiagaraMeshRendererProperties]
  renderers [{rendererClass=&NiagaraMeshRendererProperties, rendererIndex=0}]
 }
 Flat_RippleLWE {bEnabled=false, rendererClasses [], renderers []}
 Flat_RippleLWE001 {bEnabled=false, rendererClasses [], renderers []}
}
summary {
 systemName=NS_Emote_BitterSweet_RipplesRETIRED
 emitters [
  {bEnabled=false, emitterName=Flat_Ripple, rendererClasses [&NiagaraMeshRendererProperties]}
  {bEnabled=false, emitterName=Flat_Ripple001, rendererClasses [&NiagaraMeshRendererProperties]}
  {bEnabled=false, emitterName=Flat_RippleLWE, rendererClasses []}
  {bEnabled=false, emitterName=Flat_RippleLWE001, rendererClasses []}
 ]
 userVariables [
  User.WaterBodyIndex {type=NiagaraFloat, defaultValue=0, description=""}
  User.WaterDepth {type=NiagaraFloat, defaultValue=0, description=""}
  User.WaterPlaneLocation {type=Vector3f, description="", defaultValue {struct=Vector3f, value=<0 0 0>}}
  User.WaterPlaneNormal {type=Vector3f, description="", defaultValue {struct=Vector3f, value=<0 0 0>}}
  User.WaveAttenuationFactor {type=NiagaraFloat, defaultValue=0, description=""}
  User.WaveTime {type=NiagaraFloat, defaultValue=0, description=""}
 ]
}
